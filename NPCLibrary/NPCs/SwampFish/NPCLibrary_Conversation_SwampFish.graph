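# Conversation database for the SwampFish NPC, giving its bog-themed greeting and access to shared NPC conversations.
# The entry point leads to one Speech node whose default line can be replaced by contextual lines for a first-ever conversation, match conversation count, or the TeriyakiFish outfit.
# First-meeting and outfit lines use named priority curve rows, while the three count-based variants set none.
# The speech then branches through link nodes into the shared Quests and BasicServices entries.
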ConversationDatabase /NPCLibrary/NPCs/SwampFish/NPCLibrary_Conversation_SwampFish
entry_points [{tag=Conversation.Entry.SwampFish, destinations [#1]}]
nodes [
 ConversationEntryPointNode_7 {
  id=#1
  class=ConversationEntryPointNode
  outputs [#2]
  properties {entryTag {tagName=Conversation.Entry.SwampFish}}
 }
 Speech_C_4 {
  id=#2
  class=Speech_C
  outputs [#3, #4]
  properties {
   nodeName=""
   generalConfig {
    defaultMessage=bog.... bog.... bog.... bog...?
    contextualMessages [
     {
      message=The bog... welcome... you...
      contextRequirements [{requirement {class=FortControllerRequirement_IsFirstEverConversationWithNPC}}]
      priority {curve {rowName=Tandem.Speech.Priority.FirstEverInteraction}}
     }
     {
      message=Bog....?
      contextRequirements [{requirement {
         class=FortControllerRequirement_CheckNPCMatchConversationCount
         properties {nPCUniqueIdTag {tagName=AISpawnerData.Type.Specialist.Scout.Swampfish}, requiredConversationCountConfig {requirement=GreaterThan}}
        }}]
      priority {curve {curveTable=null, rowName=None}}
     }
     {
      message=Warm... bog...
      contextRequirements [{requirement {
         class=FortControllerRequirement_CheckNPCMatchConversationCount
         properties {nPCUniqueIdTag {tagName=AISpawnerData.Type.Specialist.Scout.Swampfish}, requiredConversationCountConfig {requirement=GreaterThan}}
        }}]
      priority {curve {curveTable=null, rowName=None}}
     }
     {
      message=Bog... stick...
      contextRequirements [{requirement {
         class=FortControllerRequirement_CheckNPCMatchConversationCount
         properties {nPCUniqueIdTag {tagName=AISpawnerData.Type.Specialist.Scout.Swampfish}, requiredConversationCountConfig {requirement=GreaterThan}}
        }}]
      priority {curve {curveTable=null, rowName=None}}
     }
     {
      message=No bog... so empty...
      contextRequirements [{requirement {
         class=FortControllerRequirement_HasCID
         properties {allowedCIds [&CID_315_Athena_Commando_M_TeriyakiFish]}
        }}]
      priority {curve {rowName=Tandem.Speech.Priority.OutfitReaction}}
     }
    ]
   }
  }
 }
 ConversationLinkNode_0 {id=#4, class=ConversationLinkNode, properties {remoteEntryTag {tagName=Conversation.Entry.BasicServices}}}
 ConversationLinkNode_1 {id=#3, class=ConversationLinkNode, properties {remoteEntryTag {tagName=Conversation.Entry.NPCLibrary.Quests}}}
]
